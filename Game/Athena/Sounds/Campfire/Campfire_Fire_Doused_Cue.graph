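# Campfire doused sound cue for Athena, named as the effect of a campfire's fire being put out.
# A random node picks with equal weight between two doused sound wave variations.
# Playback is routed to the gadgets and items sound class, uses the shared Athena traps attenuation asset, and has a boosted volume multiplier.
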SoundCue /Game/Athena/Sounds/Campfire/Campfire_Fire_Doused_Cue
root=Random_1
nodes [
 {id=Random_1, children [WavePlayer_0,WavePlayer_4], properties {weights [1,1]}}
 {id=WavePlayer_0, properties {soundWaveAssetPtr=&/sfx_campfire_doused_02}}
 {id=WavePlayer_4, properties {soundWaveAssetPtr=&/sfx_campfire_doused_01}}
]
properties {
 attenuationSettings=/Game/Sounds/Attenuation/Fort_Traps_Athena_Att
 maxDistance=1200
 soundClassObject=&Fort_Gadgets_Items
 volumeMultiplier=3
}
