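# Sound cue for the Athena campfire's looping fire sound, in the gadgets and items sound class with traps attenuation.
# A quality level node sends its first two levels to a random pick between two looping waves, and its last two to a single "opt" loop.
# Both branches pass through an enveloper that ramps volume from silence to full over one second, fading the loop in.
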SoundCue /Game/Athena/Sounds/Campfire/Campfire_Fire_Loop_Cue
root=QualityLevel_0
nodes [
 {
  id=Enveloper_0
  children [Random_0]
  properties {pitchCurve {editorCurveData {keys [{time=0}]}}, volumeCurve {editorCurveData {keys [<at 0 0>,<at 1 1>]}}}
 }
 {
  id=Enveloper_1
  children [WavePlayer_0]
  properties {pitchCurve {editorCurveData {keys [{time=0}]}}, volumeCurve {editorCurveData {keys [<at 0 0>,<at 1 1>]}}}
 }
 {id=QualityLevel_0, children [Enveloper_0,Enveloper_0,Enveloper_1,Enveloper_1]}
 {id=Random_0, children [WavePlayer_1,WavePlayer_2], properties {weights [1,1]}}
 {id=WavePlayer_0, properties {bLooping=true, soundWaveAssetPtr=&/sfx_campfire_loop_01_opt}}
 {id=WavePlayer_1, properties {bLooping=true, soundWaveAssetPtr=&/sfx_campfire_loop_01}}
 {id=WavePlayer_2, properties {bLooping=true, soundWaveAssetPtr=&/sfx_campfire_loop_02}}
]
properties {
 attenuationSettings=/Game/Sounds/Attenuation/Fort_Traps_Athena_Att
 maxDistance=1200
 soundClassObject=&Fort_Gadgets_Items
 volumeMultiplier=2
}
