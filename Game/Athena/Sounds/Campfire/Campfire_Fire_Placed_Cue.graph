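# Sound cue for the placement sound of a campfire in the Athena mode.
# A quality level node at the root selects one of two waves: the standard placement sound for the first two levels, the "opt" variant for the last two.
# It plays with boosted volume in the gadgets and items sound class, using the shared Athena traps attenuation to limit audible range.
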SoundCue /Game/Athena/Sounds/Campfire/Campfire_Fire_Placed_Cue
root=QualityLevel_0
nodes [
 {id=QualityLevel_0, children [WavePlayer_3,WavePlayer_3,WavePlayer_4,WavePlayer_4]}
 {id=WavePlayer_3, properties {soundWaveAssetPtr=&/sfx_campfire_place}}
 {id=WavePlayer_4, properties {soundWaveAssetPtr=&/sfx_campfire_place_opt}}
]
properties {
 attenuationSettings=/Game/Sounds/Attenuation/Fort_Traps_Athena_Att
 maxDistance=1200
 soundClassObject=&Fort_Gadgets_Items
 volumeMultiplier=2
}
