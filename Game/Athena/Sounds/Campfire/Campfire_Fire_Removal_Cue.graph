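# Sound cue for the removal of a campfire in Athena, judging by its name, built from reused small wood shatter impact sounds.
# A Random node picks one of two equally weighted shatter waves, and a Modulator randomly raises its pitch for a lighter, varied break.
# The cue plays at reduced volume through the Athena trap attenuation and the gadget items sound class.
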SoundCue /Game/Athena/Sounds/Campfire/Campfire_Fire_Removal_Cue
root=Modulator_24
nodes [
 {id=Modulator_24, children [Random_42], properties {pitchMax=1.5, pitchMin=1.2}}
 {id=Random_42, children [WavePlayer_139, WavePlayer_140], properties {weights [1,1]}}
 {id=WavePlayer_139, properties {soundWaveAssetPtr=/Game/Sounds/Fort_Impact_Sounds/Demolition/Wood/Wood_Small_Shatter01}}
 {id=WavePlayer_140, properties {soundWaveAssetPtr=/Game/Sounds/Fort_Impact_Sounds/Demolition/Wood/Wood_Small_Shatter02}}
]
properties {
 attenuationSettings=/Game/Sounds/Attenuation/Fort_Traps_Athena_Att
 maxDistance=1200
 soundClassObject=&Fort_Gadgets_Items
 volumeMultiplier=0.5
}
