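# Campfire_UI_Select_Cue is a sound cue for the campfire UI select sound.
# A single wave player feeds the Campfire_UI_Select wave straight to the output, with no randomization or modulation.
# It is routed through the non-critical in-game UI sound class at a reduced volume.
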SoundCue /Game/Athena/Sounds/Campfire/Campfire_UI_Select_Cue
root=WavePlayer_0
nodes [{id=WavePlayer_0, properties {soundWaveAssetPtr=&/Campfire_UI_Select}}]
properties {maxDistance=2097152, soundClassObject=&Fort_UI_InGame_NonCritical, volumeMultiplier=0.3}
